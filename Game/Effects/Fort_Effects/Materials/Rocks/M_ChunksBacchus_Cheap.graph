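# Cheap unlit translucent particle sprite material for rock chunk effects, shaped by the red channel of a fragment-bits texture.
# Particle color RGB feeds the shared emissive adjustment function, and particle alpha multiplies the texture mask.
# Particle speed over LerpBlurSpeed, clamped, raises a mask multiplier as speed increases, while the texture lerp blends identical inputs.
# A shading path switch gives mobile the raw mask, skipping the speed modulation.
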Material /Game/Effects/Fort_Effects/Materials/Rocks/M_ChunksBacchus_Cheap
graph {
 properties {
  bUsedWithNanite=true
  bUsedWithNiagaraSprites=true
  bUsedWithParticleSprites=true
  blendMode=BLEND_Translucent
  pixelDepthOffsetMode=PDOM_Legacy
  shadingModel=MSM_Unlit
  usedShadingModels=Unlit
 }
 nodes [
  n1 = ParticleColor()
  MaterialFunctionCall(Emissive=@n1.RGB) {materialFunction=&MF_FX_EmissiveAdjustments}
  n2 = TextureSample() {texture=/Game/Effects/Fort_Effects/Textures/Pickup/DimensionalFrag/T_Fragment_Bits}
  n3 = Clamp(None=Divide(A=ParticleSpeed(), B=ScalarParameter() {defaultValue=75, parameterName=LerpBlurSpeed}))
  n4 = LinearInterpolate(A=@n2.R, B=@n2.R, Alpha=@n3)
  n5 = LinearInterpolate(Alpha=@n3) {constA=0.65}
  Multiply(A=@n1.A, B=ShadingPathSwitch(Default=Multiply(A=@n4, B=@n5), Mobile=@n2.R) {desc=Migrated from FeatureLevelSwitch})
 ]
}
