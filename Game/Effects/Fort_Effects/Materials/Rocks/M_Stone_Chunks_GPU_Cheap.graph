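# Cheap unlit translucent sprite material for particle stone chunks, with responsive AA enabled.
# Particle random is quantized into frame steps and drives MF_SubUV to pick a chunk from a 2x2 texture grid.
# A shading path switch gives mobile simple half-tiled UVs instead, showing a single cell.
# Texture red tints particle color through MF_FX_EmissiveAdjustments, while green multiplies particle alpha.
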Material /Game/Effects/Fort_Effects/Materials/Rocks/M_Stone_Chunks_GPU_Cheap
graph {
 properties {
  bEnableResponsiveAA=true
  bTangentSpaceNormal=false
  bUsedWithParticleSprites=true
  blendMode=BLEND_Translucent
  pixelDepthOffsetMode=PDOM_Legacy
  shadingModel=MSM_Unlit
  usedShadingModels=Unlit
 }
 nodes [
  n1 = ParticleColor()
  n2 = ScalarParameter() {defaultValue=2, parameterName=Grid}
  n3 = Constant() {r=1}
  n4 = Multiply(A=ParticleRandom(), B=ScalarParameter() {defaultValue=4, parameterName=Mult_NrFrame})
  n5 = Clamp(None=Multiply(A=Ceil(None=@n4), B=ScalarParameter() {defaultValue=0.25, parameterName=Mult_NrFrame_Divide}))
  n6 = TextureSampleParameter2D(UVs=ShadingPathSwitch(
    Default=MaterialFunctionCall!(U_Images=@n2, V_Images=@n2, SubUV Texture RGB, Time Speed=@n3, Manual Time Control, Time=@n5, UVs) {materialFunction=/Game/Effects/Fort_Effects/Materials/_Functions/MF_SubUV}.UVs
    Mobile=TextureCoordinate() {uTiling=0.5, vTiling=0.5}
   ) {desc=Migrated from FeatureLevelSwitch}) {parameterName=Color and Alpha, texture=/Game/Effects/Fort_Effects/Textures/Rocks/T_Chunks_2x2_D}
  MaterialFunctionCall(Emissive=Multiply(A=@n1.RGB, B=@n6.R)) {materialFunction=&MF_FX_EmissiveAdjustments}
  Multiply(A=@n1.A, B=@n6.G)
 ]
}
